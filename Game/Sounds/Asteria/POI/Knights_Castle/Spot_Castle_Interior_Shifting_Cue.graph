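# Sound cue for one-shot ambient shifting sounds inside the Knights Castle point of interest, using indoor ambient sound class and shared tile-ambient attenuation.
# A quality-level switch sends its first two slots to an equal-weight random pool of twelve waves and its last two to a pool of four.
# A modulator at the root randomizes pitch and volume on each play, with ranges extended downward for variation.
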SoundCue /Game/Sounds/Asteria/POI/Knights_Castle/Spot_Castle_Interior_Shifting_Cue
root=Modulator_0
nodes [
 {id=Modulator_0, children [QualityLevel_0], properties {pitchMin=0.8, volumeMin=0.85}}
 {id=QualityLevel_0, children [Random_0,Random_0,Random_1,Random_1]}
 {
  id=Random_0
  children [
   WavePlayer_14
   WavePlayer_15
   WavePlayer_16
   WavePlayer_17
   WavePlayer_18
   WavePlayer_19
   WavePlayer_20
   WavePlayer_22
   WavePlayer_23
   WavePlayer_24
   WavePlayer_25
   WavePlayer_26
  ]
  properties {weights [1,1,1,1,1,1,1,1,1,1,1,1]}
 }
 {id=Random_1, children [WavePlayer_28, WavePlayer_29, WavePlayer_30, WavePlayer_31], properties {weights [1,1,1,1]}}
 {id=WavePlayer_14, properties {soundWaveAssetPtr=&/Spot_Castle_Interior_Shifting_14}}
 {id=WavePlayer_15, properties {soundWaveAssetPtr=&/Spot_Castle_Interior_Shifting_01}}
 {id=WavePlayer_16, properties {soundWaveAssetPtr=&/Spot_Castle_Interior_Shifting_02}}
 {id=WavePlayer_17, properties {soundWaveAssetPtr=&/Spot_Castle_Interior_Shifting_03}}
 {id=WavePlayer_18, properties {soundWaveAssetPtr=&/Spot_Castle_Interior_Shifting_04}}
 {id=WavePlayer_19, properties {soundWaveAssetPtr=&/Spot_Castle_Interior_Shifting_05}}
 {id=WavePlayer_20, properties {soundWaveAssetPtr=&/Spot_Castle_Interior_Shifting_06}}
 {id=WavePlayer_22, properties {soundWaveAssetPtr=&/Spot_Castle_Interior_Shifting_08}}
 {id=WavePlayer_23, properties {soundWaveAssetPtr=&/Spot_Castle_Interior_Shifting_09}}
 {id=WavePlayer_24, properties {soundWaveAssetPtr=&/Spot_Castle_Interior_Shifting_10}}
 {id=WavePlayer_25, properties {soundWaveAssetPtr=&/Spot_Castle_Interior_Shifting_11}}
 {id=WavePlayer_26, properties {soundWaveAssetPtr=&/Spot_Castle_Interior_Shifting_12}}
 {id=WavePlayer_28, properties {soundWaveAssetPtr=&Spot_Castle_Interior_Shifting_01}}
 {id=WavePlayer_29, properties {soundWaveAssetPtr=&Spot_Castle_Interior_Shifting_02}}
 {id=WavePlayer_30, properties {soundWaveAssetPtr=&Spot_Castle_Interior_Shifting_03}}
 {id=WavePlayer_31, properties {soundWaveAssetPtr=&Spot_Castle_Interior_Shifting_04}}
]
properties {
 attenuationSettings=/Game/Sounds/Attenuation/Fort_TileAmbient_OneShot_LessPositional_Log_Att
 maxDistance=6200
 soundClassObject=/Game/Sounds/FortSoundClassesAndModes/Fort_Ambient_OneShots_Indoors
 volumeMultiplier=0.401
}
